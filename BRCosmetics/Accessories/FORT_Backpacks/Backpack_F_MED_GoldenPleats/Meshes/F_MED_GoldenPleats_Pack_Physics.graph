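# Physics asset for the GoldenPleats backpack cosmetic, defining bodies and joints for its dangling belt and cloth bones.
# Two chains, a three-bone belt and a six-bone cloth strip, each hang from a shapeless MSTR root body.
# Every chain bone carries a same-sized small capsule and is constrained to the bone above it.
# Tight per-joint swing and twist limits restrict each link to small bends with very little twist.
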PhysicsAsset /BRCosmetics/Accessories/FORT_Backpacks/Backpack_F_MED_GoldenPleats/Meshes/F_MED_GoldenPleats_Pack_Physics
bodies {
 dyn_belt_01 [{shapeType=Capsule, length=4, radius=2, shapeName=dyn_belt_01_capsule, center=<1 0 0>, rotation=<rot 90 0 0>}]
 dyn_belt_02 [{shapeType=Capsule, length=4, radius=2, shapeName=dyn_belt_02_capsule, center=<1 0 0>, rotation=<rot 90 0 0>}]
 dyn_belt_03 [{shapeType=Capsule, length=4, radius=2, shapeName=dyn_belt_03_capsule, center=<2 0 0>, rotation=<rot 90 0 0>}]
 dyn_belt_MSTR []
 dyn_cloth_01 [{shapeType=Capsule, length=4, radius=2, shapeName=dyn_cloth_01_capsule, center=<1 0 0>, rotation=<rot 90 0 0>}]
 dyn_cloth_02 [{shapeType=Capsule, length=4, radius=2, shapeName=dyn_cloth_02_capsule, center=<1 0 0>, rotation=<rot 90 0 0>}]
 dyn_cloth_03 [{shapeType=Capsule, length=4, radius=2, shapeName=dyn_cloth_03_capsule, center=<1 0 0>, rotation=<rot 90 0 0>}]
 dyn_cloth_04 [{shapeType=Capsule, length=4, radius=2, shapeName=dyn_cloth_04_capsule, center=<1 0 0>, rotation=<rot 90 0 0>}]
 dyn_cloth_05 [{shapeType=Capsule, length=4, radius=2, shapeName=dyn_cloth_05_capsule, center=<1 0 0>, rotation=<rot 90 0 0>}]
 dyn_cloth_06 [{shapeType=Capsule, length=4, radius=2, shapeName=dyn_cloth_06_capsule, center=<1 0 0>, rotation=<rot 90 0 0>}]
 dyn_cloth_MSTR []
}
constraints [
 {bone1Name=dyn_belt_01, bone2Name=dyn_belt_MSTR, swing1LimitDegrees=6, swing2LimitDegrees=15, twistLimitDegrees=1}
 {bone1Name=dyn_belt_02, bone2Name=dyn_belt_01, swing1LimitDegrees=20, swing2LimitDegrees=4, twistLimitDegrees=4}
 {bone1Name=dyn_belt_03, bone2Name=dyn_belt_02, swing1LimitDegrees=20, swing2LimitDegrees=4, twistLimitDegrees=5}
 {bone1Name=dyn_cloth_01, bone2Name=dyn_cloth_MSTR, swing1LimitDegrees=16, swing2LimitDegrees=8, twistLimitDegrees=8}
 {bone1Name=dyn_cloth_02, bone2Name=dyn_cloth_01, swing1LimitDegrees=16, swing2LimitDegrees=6, twistLimitDegrees=1}
 {bone1Name=dyn_cloth_03, bone2Name=dyn_cloth_02, swing1LimitDegrees=18, swing2LimitDegrees=12, twistLimitDegrees=5}
 {bone1Name=dyn_cloth_04, bone2Name=dyn_cloth_03, swing1LimitDegrees=20, swing2LimitDegrees=10, twistLimitDegrees=5}
 {bone1Name=dyn_cloth_05, bone2Name=dyn_cloth_04, swing1LimitDegrees=20, swing2LimitDegrees=10, twistLimitDegrees=5}
 {bone1Name=dyn_cloth_06, bone2Name=dyn_cloth_05, swing1LimitDegrees=20, swing2LimitDegrees=10, twistLimitDegrees=5}
]
